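# Material function for LavaPit effect materials that outputs a fluid fill mask, a two-sided mask and a fluid surface world position.
# World position, relative to object origin plus an offset, is rotated into a fluid-plane basis, shifted by a waterline term with turbulence input, then thresholded.
# VirtualPlaneCoordinates supplies the surface position, and TwoSidedSign the face mask.
# A Niagara static switch selects instance transforms and dynamic-parameter plane axes over object-based fallbacks.
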MaterialFunction /SparksMaps/Effects/Environment/S4/LavaPit/MF_Sparks_FluidMasks
graph {nodes [
  FunctionOutput(None=NamedRerouteUsage()) {outputName=TwosidedMask}
  FunctionOutput(None=NamedRerouteUsage()) {outputName=FluidMask}
  n1 = Add(A=NamedRerouteUsage(), B=NamedRerouteUsage())
  n2 = Multiply(A=NamedRerouteUsage(), B=NamedRerouteUsage())
  FunctionOutput(None=MaterialFunctionCall(Plane Normal Axis=NamedRerouteUsage(), Plane X Axis=NamedRerouteUsage(), Plane Center=Add(A=@n1, B=@n2), UV Scale=Constant() {r=1}) {materialFunction=&VirtualPlaneCoordinates}.World Position) {outputName=FluidSurfaceWP}
  n3 = PreSkinnedLocalBounds()
  NamedRerouteDeclaration(None=@n3.Extents) {name=Extents, nodeColor=<rgba 0.5 1 0 1>}
  n4 = Multiply(A=Add(
    A=NamedRerouteDeclaration(None=@n3.Min) {name=MinBnds, nodeColor=<rgba 0.2891 1 0 1>}
    B=NamedRerouteDeclaration(None=@n3.Max) {name=MaxBnds, nodeColor=<rgba 0 1 0.3438 1>}
   )) {constB=0.5}
  n5 = NamedRerouteDeclaration(None=Multiply(
    A=NamedRerouteUsage()
    B=MaterialFunctionCall() {materialFunction=&ObjectScale}.Scale XYZ
   )) {name=Obj_Dimensions, nodeColor=<rgba 0.4531 1 0 1>}
  NamedRerouteDeclaration(None=StaticSwitchParameter(True=Transform(None=@n4) {transformSourceType=TRANSFORMSOURCE_Instance}, False=Multiply(A=@n5) {constB=0.5}) {defaultValue=true, group=FluidMask, parameterName=Niagara}) {name=CenterOffset, nodeColor=<rgba 0 0.4844 1 1>}
  n6 = DynamicParameter() {parameterIndex=2, defaultValue=<rgba 0 0 0 0>, paramNames [Speed,Turbulence,Param3,Param4]}
  NamedRerouteDeclaration(None=@n6.Speed) {name=Speed, nodeColor=<rgba 0 0.7891 1 1>}
  n7 = DynamicParameter() {defaultValue=<rgba 0 0 1 0.5>, paramNames [Water Normal X, Water Normal Y, Water Normal Z, Fill]}
  n8 = MaterialFunctionCall(X=@n7.Water Normal X, Y=@n7.Water Normal X, Z=@n7.Water Normal X) {materialFunction=&MakeFloat3}
  NamedRerouteDeclaration(None=StaticSwitchParameter(True=Normalize(VectorInput=@n8.Result), False=Constant3Vector() {constant=<rgba 0 0 1 0>}) {defaultValue=true, group=FluidMask, parameterName=Niagara}) {name=WaterPlaneZ, nodeColor=<rgba 1 0 0.1875 1>}
  NamedRerouteDeclaration(None=@n6.Turbulence) {name=Turbulence, nodeColor=<rgba 0 0.7891 1 1>}
  NamedRerouteDeclaration(None=ScalarParameter() {group=FluidMask, parameterName=FillAmount}) {name=Fill, nodeColor=<rgba 0 1 0.0859 1>}
  n9 = Multiply(A=NamedRerouteUsage(), B=Constant3Vector() {constant=<rgba 1 1 -1 0>})
  n10 = Multiply(A=NamedRerouteUsage(), B=Constant3Vector() {constant=<rgba -1 -1 1 0>})
  n11 = WorldPosition() {worldPositionShaderOffset=WPT_ExcludeAllShaderOffsets}
  n12 = Constant3Vector()
  n13 = TransformPosition(None=@n12) {transformSourceType=TRANSFORMPOSSOURCE_Instance, transformType=TRANSFORMPOSSOURCE_World}
  n14 = TransformPosition(None=@n12) {transformType=TRANSFORMPOSSOURCE_World}
  n15 = Add(
   A=NamedRerouteDeclaration(None=StaticSwitchParameter(True=@n13, False=@n14) {defaultValue=true, group=FluidMask, parameterName=Niagara}) {name=ObjPosition-WS, nodeColor=<rgba 0 0.1562 1 1>}
   B=NamedRerouteUsage()
  )
  n16 = NamedRerouteDeclaration(None=MaterialFunctionCall(VectorToTransform=Subtract(A=@n11.XYZ, B=@n15), BasisX=@n9, BasisY=CrossProduct(A=@n10, B=@n9), BasisZ=@n10) {materialFunction=&Transform3x3Matrix}.Transformed Vector) {name=WorldPosition, nodeColor=<rgba 0 1 0.25 1>}
  n17 = Constant()
  n18 = Multiply(
   A=Add(A=NamedRerouteUsage(), B=FunctionInput(Preview=@n17) {bUsePreviewValueAsDefault=true, inputName=Turbulance, inputType=FunctionInput_Scalar})
   B=NamedRerouteUsage()
  )
  n19 = Saturate(None=NamedRerouteDeclaration(None=Add(A=@n16, B=@n18)) {name=WaterLine, nodeColor=<rgba 1 0 0.2344 1>})
  NamedRerouteDeclaration(None=ComponentMask(None=OneMinus(None=@n19)) {b=true}) {name=WaterMask, nodeColor=<rgba 1 0 0.6562 1>}
  n20 = DynamicParameter() {parameterIndex=1, defaultValue=<rgba 1 0 0 0>, paramNames [Tangent X, Tangent Y, Tangent Z, Time]}
  NamedRerouteDeclaration(None=StaticSwitchParameter(
    True=MaterialFunctionCall(X=@n20.Tangent X, Y=@n20.Tangent X, Z=@n20.Tangent X) {materialFunction=&MakeFloat3}.Result
    False=Constant3Vector() {constant=<rgba 1 0 0 0>}
   ) {defaultValue=true, group=FluidMask, parameterName=Niagara}) {name=WaterPlaneTangent, nodeColor=<rgba 0.8984 1 0 1>}
  NamedRerouteDeclaration(None=Saturate(None=TwoSidedSign())) {name=TwoSidedMask, nodeColor=<rgba 1 0 0.4453 1>}
  NamedRerouteDeclaration(None=@n20.Time) {name=Niagara Time, nodeColor=<rgba 0 1 0.6719 1>}
  NamedRerouteDeclaration(None=@n6.Param3) {name=Foam Time, nodeColor=<rgba 0 1 0.7656 1>}
  ObjectLocalBounds()
  StaticSwitchParameter(True=@n7.Fill, False=ScalarParameter() {defaultValue=1, group=Static Overrides, parameterName=Liquid Fill}) {defaultValue=true, parameterName=Is Niagara}
  VectorParameter() {parameterName=CenterOffset, defaultValue=<rgba 0 0 0 1>}
 ]}
