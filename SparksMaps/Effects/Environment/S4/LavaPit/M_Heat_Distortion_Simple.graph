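# M_Heat_Distortion_Simple is a translucent unlit heat-distortion material for LavaPit environment effects, enabled for sprite, mesh and ribbon particles.
# A vertically panning wispy fire noise normal blends from flat through a soft radial UV mask, fading toward the edges.
# A dynamic parameter sets distortion amount by interpolating a value for index-of-refraction distortion.
# A particle color node is present, though the file shows no output connections.
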Material /SparksMaps/Effects/Environment/S4/LavaPit/M_Heat_Distortion_Simple
graph {
 properties {
  bAutomaticallySetUsageInEditor=false
  bUsedWithMeshParticles=true
  bUsedWithNiagaraMeshParticles=true
  bUsedWithNiagaraRibbons=true
  bUsedWithNiagaraSprites=true
  bUsedWithParticleSprites=true
  blendMode=BLEND_Translucent
  pixelDepthOffsetMode=PDOM_Legacy
  refractionMethod=RM_IndexOfRefraction
  shadingModel=MSM_Unlit
  usedShadingModels=Unlit
 }
 nodes [
  Constant()
  n1 = TextureSampleParameter2D(UVs=Panner() {speedY=0.25}) {parameterName=Noise Normal, samplerType=SAMPLERTYPE_Normal, texture=/Game/Effects/Fort_Effects/Textures/Fire/T_WispyFireNoise_Soft_N}
  n2 = SphereMask(A=TextureCoordinate(), B=Constant2Vector() {g=0.5, r=0.5}) {attenuationRadius=0.5, hardnessPercent=0}
  LinearInterpolate(A=Constant3Vector() {constant=<rgba 0 0 1 0>}, B=@n1.RGB, Alpha=Saturate(None=Power(Base=@n2)))
  LinearInterpolate(Alpha=DynamicParameter() {paramNames ["Distortion Amount (0-1)", Param2, Param3, Param4]}."Distortion Amount (0-1)") {constA=1, constB=1.5}
  ParticleColor()
 ]
}
