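# Unlit additive particle material that draws a soft procedural radial glow, filed with the LavaPit environment effects.
# A radial gradient with a Hardness exponent is tinted by particle color and passed through the shared FX emissive adjustments function.
# Particle alpha times that gradient drives a depth-faded opacity chain, with optional near-camera and view-angle mesh fades.
# Both optional fades blend in through scalar lerp parameters rather than static switches.
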Material /SparksMaps/Effects/Environment/S4/LavaPit/M_Sparks_RadialGlow_Additive
graph {
 properties {
  bUsedWithNanite=true
  bUsedWithNiagaraSprites=true
  bUsedWithParticleSprites=true
  blendMode=BLEND_Additive
  pixelDepthOffsetMode=PDOM_Legacy
  shadingModel=MSM_Unlit
  usedShadingModels=Unlit
 }
 nodes [
  n1 = ParticleColor()
  n2 = Power(
   Base=MaterialFunctionCall!(Radius, CenterPosition) {materialFunction=/Engine/Functions/Engine_MaterialFunctions01/Gradient/RadialGradient}.RadialGradient
   Exp=ScalarParameter() {defaultValue=1, parameterName=Hardness}
  ) {constExponent=5}
  n3 = DepthFade(Opacity=Multiply(A=@n1.A, B=@n2), FadeDistance=ScalarParameter() {defaultValue=500, group=DepthFade, parameterName=FadeDistance}) {fadeDistanceDefault=500}
  n4 = ScalarParameter() {defaultValue=1024, parameterName=NearCameraFade}
  n5 = LinearInterpolate(
   A=@n3
   B=Multiply(A=@n3, B=MaterialFunctionCall(Distance=@n4) {materialFunction=&Near_Camera_Fade}.Result)
   Alpha=ScalarParameter() {parameterName=Use NearCameraFade}
  )
  n6 = CameraVectorWS()
  n7 = VertexNormalWS()
  n8 = Power(Base=Abs(None=DotProduct(A=@n6, B=@n7)), Exp=ScalarParameter() {defaultValue=1, parameterName=MeshFade Power})
  LinearInterpolate(A=@n5, B=Multiply(A=@n5, B=Saturate(None=@n8)), Alpha=ScalarParameter() {parameterName=Use MeshFade})
  MaterialFunctionCall(Emissive=Multiply(A=@n1.RGB, B=@n2)) {materialFunction=&MF_FX_EmissiveAdjustments}
 ]
}
